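# Translucent UI material that draws the Ancient Weapon reticle, with a look that changes when an enemy is in range.
# Center, bracket and glyph shapes come from one channel-packed distance-field texture, while procedural boxes and gradients add a box and light lines.
# Mirrored UVs with parameterized X offsets position the side elements.
# IsEnemyInRange blends color and opacity layers, with glyph opacity and final alpha exposed as parameters.
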Material /AncientWeapon/Reticle/Materials/M_UI_AncientWeapon_Reticle
graph {
 properties {blendMode=BLEND_Translucent, materialDomain=MD_UI, pixelDepthOffsetMode=PDOM_Legacy, usedShadingModels=Default Lit}
 nodes [
  n1 = NamedRerouteUsage()
  n2 = NamedRerouteUsage()
  n3 = VectorParameter() {parameterName=Color, defaultValue=<rgba 1 1 1 1>}
  n4 = NamedRerouteUsage()
  n5 = Constant3Vector()
  n6 = LinearInterpolate(
   A=NamedRerouteUsage()
   B=LinearInterpolate(A=Multiply(A=@n2, B=@n3.RGB), B=Multiply(A=@n4, B=@n5), Alpha=@n4)
   Alpha=ScalarParameter() {defaultValue=1, parameterName=IsEnemyInRange}
  )
  n7 = NamedRerouteUsage()
  n8 = LinearInterpolate(
   A=VectorParameter() {parameterName=Color1, defaultValue=<rgba 0.521 0.2384 1 1>}.RGB
   B=VectorParameter() {parameterName=Color2, defaultValue=<rgba 0.7625 0.1718 0.8228 1>}.RGB
   Alpha=MaterialFunctionCall() {materialFunction=&LinearGradient}.VGradient
  )
  Add(
   A=Add(A=Add(A=@n1, B=@n6), B=NamedRerouteUsage())
   B=LinearInterpolate(B=Multiply(A=@n7, B=@n8), Alpha=ScalarParameter() {defaultValue=1, parameterName=IsEnemyInRange})
  )
  n9 = NamedRerouteUsage()
  n10 = NamedRerouteUsage()
  n11 = Add(
   A=NamedRerouteUsage()
   B=LinearInterpolate(
    A=NamedRerouteUsage()
    B=LinearInterpolate(A=@n9, B=@n10, Alpha=@n10)
    Alpha=ScalarParameter() {defaultValue=1, parameterName=IsEnemyInRange}
   )
  )
  n12 = LinearInterpolate(
   B=Multiply(A=NamedRerouteUsage(), B=ScalarParameter() {defaultValue=0.3, parameterName=GlyphsOpacity})
   Alpha=ScalarParameter() {defaultValue=1, parameterName=IsEnemyInRange}
  )
  n13 = NamedRerouteUsage()
  n14 = ScalarParameter() {defaultValue=1, parameterName=IsEnemyInRange}
  Multiply(A=Add(A=Add(A=@n11, B=@n12), B=LinearInterpolate(B=@n13, Alpha=@n14)), B=ScalarParameter() {defaultValue=1, parameterName=FinalAlpha})
  n15 = TextureSample(UVs=MaterialFunctionCall(UVs=TextureCoordinate(), Texture Scale=Constant() {r=0.1417}) {materialFunction=&ScaleUVsByCenter}.UVs) {
   samplerSource=SSM_Clamp_WorldGroupSettings
   samplerType=SAMPLERTYPE_LinearColor
   texture=/AncientWeapon/Reticle/Textures/T_UI_AncientWeaponReticle_DF
  }
  NamedRerouteDeclaration(None=MaterialFunctionCall!(RGBA Channels=@n15.G, Glow Min=Constant() {r=0.1}, Glow Max=Constant() {r=0.4314}, Outline Thickness) {materialFunction=/FortUILibrary/MaterialFunctions/Basics/MF_UI_ChannelPackedDistanceField}.Outline) {name=CenterOutline, nodeColor=<rgba 0 0 0 1>}
  n16 = AppendVector(A=ScalarParameter() {defaultValue=0.2011, parameterName=BracketsX}, B=Constant())
  n17 = Constant2Vector() {g=0.5, r=1}
  n18 = TextureSample(UVs=MaterialFunctionCall(
    UVs=MaterialFunctionCall!(UV, Offset=@n16, Rotation, Scale=@n17, MirrorY) {materialFunction=/FortUILibrary/MaterialFunctions/Basics/MF_UI_Symmetry_2x1_DEPRECATED}.Result
    "Rotation Angle (0-1)"=Constant() {r=-0.115}
   ) {materialFunction=&CustomRotator}.Rotated Values) {
   samplerSource=SSM_Clamp_WorldGroupSettings
   samplerType=SAMPLERTYPE_LinearColor
   texture=&T_UI_AncientWeaponReticle_DF
  }
  n19 = Constant() {r=0.6354}
  n20 = Constant() {r=0.9}
  n21 = MaterialFunctionCall(Texture Scale=Constant2Vector() {g=0.744, r=3.6469}) {materialFunction=&ScaleUVsByCenter}
  NamedRerouteDeclaration(None=Multiply(
    A=MaterialFunctionCall(RGBA Channels=@n18.B, Glow Min=@n19, Glow Max=@n20) {materialFunction=&MF_UI_ChannelPackedDistanceField}.Fill
    B=MaterialFunctionCall(UV=@n21.UVs) {materialFunction=&MF_UI_RadialGradient}.Smooth
   )) {name=Bracket, nodeColor=<rgba 0 0 0 1>}
  NamedRerouteDeclaration(None=MaterialFunctionCall(RGBA Channels=@n15.G, Glow Min=Constant() {r=0.6263}, Glow Max=Constant() {r=0.9255}) {materialFunction=&MF_UI_ChannelPackedDistanceField}.Fill) {name=CenterActive, nodeColor=<rgba 0 0 0 1>}
  n22 = AppendVector(A=ScalarParameter() {defaultValue=-0.0457, parameterName=GlyphsX}, B=Constant())
  n23 = Constant2Vector() {g=0.7103, r=1}
  n24 = TextureSample(UVs=MaterialFunctionCall(
    UVs=MaterialFunctionCall(Offset=@n22, Scale=@n23) {materialFunction=&MF_UI_Symmetry_2x1_DEPRECATED}.Result
    Texture Scale=Constant() {r=0.2697}
   ) {materialFunction=&ScaleUVsByCenter}.UVs) {
   samplerSource=SSM_Clamp_WorldGroupSettings
   samplerType=SAMPLERTYPE_LinearColor
   texture=&T_UI_AncientWeaponReticle_DF
  }
  NamedRerouteDeclaration(None=MaterialFunctionCall(RGBA Channels=@n24.R, Glow Min=Constant() {r=0.1463}, Glow Max=Constant() {r=0.9346}) {materialFunction=&MF_UI_ChannelPackedDistanceField}.Fill) {name=Glyphs, nodeColor=<rgba 0 0 0 1>}
  n25 = TextureCoordinate()
  n26 = Constant() {r=0.112}
  n27 = MaterialFunctionCall(
   UVs=MaterialFunctionCall(
    UVs=MaterialFunctionCall(UVs=@n25, "Rotation Angle (0-1)"=@n26) {materialFunction=&CustomRotator}.Rotated Values
    Texture Scale=Constant() {r=0.1417}
   ) {materialFunction=&ScaleUVsByCenter}.UVs
   OutlineThickness=Constant() {r=0.0549}
  ) {materialFunction=&MF_UI_Box}
  NamedRerouteDeclaration(None=OneMinus(None=SmoothStep(Value=@n27.SDFOutline) {constMax=0.1})) {name=Box, nodeColor=<rgba 0 0 0 1>}
  n28 = MaterialFunctionCall(Offset=Constant2Vector() {r=0.2514}) {materialFunction=&MF_UI_Symmetry_2x1_DEPRECATED}
  n29 = AppendVector(A=ScalarParameter() {defaultValue=0.2, parameterName=LightX}, B=Constant() {r=0.8})
  n30 = OneMinus(None=SmoothStep(Value=MaterialFunctionCall(UVs=@n28.Result, Size=@n29) {materialFunction=&MF_UI_Box}.SDF) {constMax=0.5})
  n31 = MaterialFunctionCall(UV=MaterialFunctionCall(Texture Scale=Constant2Vector() {g=0.5794, r=10}) {materialFunction=&ScaleUVsByCenter}.UVs) {materialFunction=&MF_UI_RadialGradient}
  n32 = MaterialFunctionCall(Offset=Constant() {r=0.3122}) {materialFunction=&MF_UI_Symmetry_2x1_DEPRECATED}
  NamedRerouteDeclaration(None=Multiply(
    A=Multiply(A=@n30, B=@n31.Smooth)
    B=MaterialFunctionCall(Texture=@n32.Result) {materialFunction=&MF_UI_LinearGradient}.Result
   )) {name=Lines, nodeColor=<rgba 0 0 0 1>}
 ]
}
