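# Reticle widget for the thrown Ancient Weapon, showing whether a pawn or the environment is in range.
# It binds to the weapon's Update Reticle State event on Construct and clears all that event's bindings on Destruct.
# Each update sets alpha and enemy-in-range parameters on the reticle's dynamic material.
# Two counters limit a material-parameter animation to play forward or reverse only on the first update after pawn range changes.
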WidgetBlueprint /AncientWeapon/Reticle/WBP_AncientWeaponReticle
animations [Anim_EnemyInRange {
  binding_count=1
  length_seconds=0.25
  object=&self:Anim_EnemyInRange
  sequence {bindings [Reticle {
     id=#1
     possessed_class=Image
     tracks [{
       type=MovieSceneWidgetMaterialTrack
       sections [{
         type=MovieSceneParameterSection
         end_frame=5
         start_frame=0
         channels [
          LightX {type=MovieSceneScriptingFloatChannel, keys [[0,-0.34],[5,0.2]]}
          IsEnemyInRange {type=MovieSceneScriptingFloatChannel, keys [[0,0],[5,1]]}
          GlyphsOpacity {type=MovieSceneScriptingFloatChannel, keys [[0,0],[5,0.5]]}
          GlyphsX {type=MovieSceneScriptingFloatChannel, keys [[0,-0.0457],[5,-0.2561]]}
          BracketsX {type=MovieSceneScriptingFloatChannel, keys [[0,0.2011],[5,0.1211]]}
         ]
        }]
      }]
    }]}
 }]
graphs [
 EventGraph {nodes [
   n1 = UpdateReticleState_Event!() -> (OutputDelegate, then, bPawnInRange:bool=~, bEnvironmentInRange:bool=~) {is=title K2Node_CustomEvent}
   > Set in Range State!(execute, self, bPlayerInRange:bool=@n1.bPawnInRange, bEnvInRange:bool=@n1.bEnvironmentInRange) -> (then) {is=title K2Node_CallFunction}
   Event Construct()
   > n2 = Cast To B_Chrono_Weap_AncientWeapon_Thrown!(
    execute
    Object=GetEquippedWeapon!(self:Fort Weapon Reticle Extension Widget Base Object Reference) -> (ReturnValue:Fort Weapon Object Reference)
   ) -> (then, CastFailed, AsB Chrono Weap Ancient Weapon Thrown:B Chrono Weap Ancient Weapon Thrown Object Reference) {is=title K2Node_DynamicCast}
   > n3 = Set As B Chrono Weap Ancient Weapon Thrown(As B Chrono Weap Ancient Weapon Thrown:B Chrono Weap Ancient Weapon Thrown Object Reference=@n2) -> (Output_Get:B Chrono Weap Ancient Weapon Thrown Object Reference)
   > Bind Event to Update Reticle State(self:B Chrono Weap Ancient Weapon Thrown Object Reference=@n3, Delegate=@n1) {is=title K2Node_AddDelegate}
   Event Destruct()
   > Unbind all Events from Update Reticle State!(
    execute:Exec
    self:B Chrono Weap Ancient Weapon Thrown Object Reference=Get As B Chrono Weap Ancient Weapon Thrown() -> (As B Chrono Weap Ancient Weapon Thrown:B Chrono Weap Ancient Weapon Thrown Object Reference)
   ) -> (then:Exec) {is=title K2Node_ClearDelegate}
  ]}
 Set in Range State {nodes [
   n1 = Set in Range State!() -> (then, bPlayerInRange:bool=~, bEnvInRange:bool=~) {is=title K2Node_FunctionEntry}
   > n2 = GetDynamicMaterial(self=Get Reticle() -> (Reticle:Image Object Reference))
   > n3 = Set ReticleMaterialReference(ReticleMaterialReference:Material Instance Dynamic Object Reference=@n2) -> (Output_Get:Material Instance Dynamic Object Reference)
   > SetScalarParameterValue(self=@n3, ParameterName=FinalAlpha, Value=SelectFloat(A=1, B=0.6, bPickA=OR Boolean(A=@n1, B=@n1.bEnvInRange)))
   > SetScalarParameterValue(
    self=Get ReticleMaterialReference() -> (ReticleMaterialReference:Material Instance Dynamic Object Reference)
    ParameterName=IsEnemyInRange
    Value=SelectFloat(A=1, B=0, bPickA=Get bPlayerInRange() -> (bPlayerInRange:bool))
   )
   > n4 = Branch(Condition=Get bPlayerInRange())
   > Set Counter_NotInRange(Counter_NotInRange:double=0) -> (Output_Get:double)
   > n5 = Increment Float!(" ":Exec, Value:"Float (double-precision) (by ref)"=Get Counter() -> (Counter:double)) -> ("  ":Exec, Result:double=~) {is=title K2Node_MacroInstance}
   > n6 = Set Counter(Counter:double=@n5) -> (Output_Get:double)
   > Branch(Condition="float <= float"(A=@n6, B=1) -> (ReturnValue=~))
   n7 = Get Anim_EnemyInRange() -> (Anim_EnemyInRange:Widget Animation Object Reference)
   > PlayAnimationForward(InAnimation=@n7)
   Set Counter(execute=@n4.else, Counter=0)
   > n8 = Increment Float(Value=Get Counter_NotInRange() -> (Counter_NotInRange:double)) -> (Result=~)
   > n9 = Set Counter_NotInRange(Counter_NotInRange=@n8)
   > Branch(Condition="float <= float"(A=@n9, B=1) -> (ReturnValue=~))
   > PlayAnimationReverse(InAnimation=@n7)
  ]}
]
variables [
 As B Chrono Weap Ancient Weapon Thrown {type=B_Chrono_Weap_AncientWeapon_Thrown_C, default=null}
 Cached Has Homing Target {type=boolean}
 Counter {type=number}
 Counter_NotInRange {type=number}
 Homing {type=WidgetAnimation, default=null}
 ReticleMaterialReference {type=MaterialInstanceDynamic}
 ShowingHoming {type=boolean}
]
widgets {
 info {
  inheritedWidgetCount=1
  parentClass=/Script/FortniteUI.FortWeaponReticleExtensionWidgetBase
  rootWidgetClass=&Overlay
  widgetCount=3
 }
 widgets [
  ReticleOverlay {widgetClassPath=&Overlay, bIsVariable=true, widget=ReticleOverlay}
  Reticle {
   widgetClassPath=&Image
   bIsVariable=true
   parent=ReticleOverlay
   widget=Reticle
   properties {brush {
     bIsDynamicallyLoaded=false
     drawAs=Image
     imageType=NoImage
     mirroring=NoMirror
     resourceName=None
     resourceObject=/AncientWeapon/Reticle/Materials/M_UI_AncientWeapon_Reticle
     tiling=NoTile
     imageSize=<64 64>
     margin=<ltrb 0 0 0 0>
     outlineSettings {
      bUseBrushTransparency=false
      roundingType=HalfHeightRadius
      width=0
      color {colorUseRule=UseColor_Specified, specifiedColor=<rgba 0 0 0 0>}
      cornerRadii=<0 0 0 0>
     }
     tintColor {colorUseRule=UseColor_Specified, specifiedColor=<rgba 1 1 1 1>}
     uVRegion {bIsValid=false, max=<0 0>, min=<0 0>}
    }}
   slot {class=OverlaySlot, properties {horizontalAlignment=HAlign_Center, verticalAlignment=VAlign_Center}}
  }
  TutorialHighlight {widgetClassPath=&TutorialHighlight}
 ]
}
